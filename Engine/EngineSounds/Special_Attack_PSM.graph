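# Special_Attack_PSM is a sound class for special attack sounds, parented to the SFX class.
# It carries a passive sound mix modifier that applies Duck_Normal_Attack_MixMod while sounds in this class play above a minimum volume threshold.
# The name of that mix suggests it ducks normal attack sounds so special attacks stand out.
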SoundClass /Engine/EngineSounds/Special_Attack_PSM
parent=&/SFX
properties {passiveSoundMixModifiers [{minVolumeThreshold=0.3, soundMix=&/Duck_Normal_Attack_MixMod}]}
